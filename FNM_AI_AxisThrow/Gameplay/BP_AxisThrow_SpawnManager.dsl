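; AxisThrow spawn manager that picks the main or an alternate Living World point provider to enable.
; On BeginPlay, where cosmetic events cannot execute, it waits for playlist data, then rolls against ChanceToSelectMain.
; A failed roll picks a random alternate, if any, activates its data layer, and enables its provider after BufferTimer.
; Cheat events force either choice, disabling providers, unloading alternate layers and running "cheat lwmf".
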(event Custom|EnableMainProvider
  (Class|FortAthenaLivingWorldPointProviderInterface|EnablePointProvider_DEPRECATED (Variables|Default|GetMainProvider)))

(event EventBeginPlay
  (bind _alternatelocationproviders (Variables|Default|GetAlternateLocationProviders))
  (bind _returnvalue (Utilities|Array|Length _alternatelocationproviders))
  (Networking|CanExecuteCosmeticEvents
    (:False
      (Class|FortAsyncActionCurrentPlaylistReady|PlaylistDataReadyAsync
        (:OnReady
          (if (<= (Math|Random|RandomFloatinRange 0.0 1.0) (ScalableFloat|GetValueatLevel (Variables|Default|GetChanceToSelectMain)))
            (Class|FortAthenaLivingWorldPointProviderInterface|EnablePointProvider_DEPRECATED (Variables|Default|GetMainProvider))
            (elif (> _returnvalue 0)
              (Variables|Default|SetSelectedIndex (- (Math|Random|RandomIntegerinRange 1 _returnvalue) 1))
              (DataLayers|SetDataLayerRuntimeStatebyLabel (WorldSubsystems|GetDataLayerSubsystem) (Utilities|Array|Get(acopy) (Variables|Default|GetAlternateLocationDataLayerNames) (Variables|Default|GetSelectedIndex)) "Activated" false)
              (Utilities|FlowControl|Delay (ScalableFloat|GetValueatLevel (Variables|Default|GetBufferTimer)))
              (Class|FortAthenaLivingWorldPointProviderInterface|EnablePointProvider_DEPRECATED (Utilities|Array|Get(acopy) (Variables|Default|GetAlternateLocationProviders) (Variables|Default|GetSelectedIndex))))))))))

(event Custom|EnableAlternateLocation
  (DataLayers|SetDataLayerRuntimeStatebyLabel (WorldSubsystems|GetDataLayerSubsystem) (Utilities|Array|Get(acopy) (Variables|Default|GetAlternateLocationDataLayerNames) (Variables|Default|GetSelectedIndex)) "Activated" false)
  (Utilities|FlowControl|Delay (ScalableFloat|GetValueatLevel (Variables|Default|GetBufferTimer)))
  (Class|FortAthenaLivingWorldPointProviderInterface|EnablePointProvider_DEPRECATED (Utilities|Array|Get(acopy) (Variables|Default|GetAlternateLocationProviders) (Variables|Default|GetSelectedIndex))))

(event Custom|DisableAlternateLocations
  (for _array_element (Variables|Default|GetAlternateLocationProviders)
    (Class|FortAthenaLivingWorldPointProviderInterface|DisablePointProvider _array_element)
    (for _array_element_2 (Variables|Default|GetAlternateLocationDataLayerNames)
      (DataLayers|SetDataLayerRuntimeStatebyLabel (WorldSubsystems|GetDataLayerSubsystem) _array_element_2 "Unloaded" false))))

(event Custom|AxisThrow_ForceSpawnMain
  ;[1 0.15 0.15 0.5] CHEATS MENU
  (CallFunction|DisableAlternateLocations)
  (CallFunction|EnableMainProvider)
  (Development|ExecuteConsoleCommand "cheat lwmf"))

(event Custom|AxisThrow_ForceSpawnAlternateLocation
  (params (LocationIndex :type Integer))
  ;[1 0.15 0.15 0.5] CHEATS MENU
  (CallFunction|DisableAlternateLocations)
  (Class|FortAthenaLivingWorldPointProviderInterface|DisablePointProvider (Variables|Default|GetMainProvider))
  (Variables|Default|SetSelectedIndex LocationIndex)
  (CallFunction|EnableAlternateLocation)
  (Utilities|FlowControl|Delay (+ (ScalableFloat|GetValueatLevel (Variables|Default|GetBufferTimer)) 0.2))
  (Development|ExecuteConsoleCommand "cheat lwmf"))

(fn ConstructionScript)
